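# Translucent, unlit, two-sided overlay material for soft glowing outlines on the SheerWill Phase player body cosmetic.
# All of its shading comes from a single material function, MF_Sheerwill_Outlines, with material attributes enabled.
# It supports skeletal meshes and mesh deformers and renders in the before-depth-of-field translucency pass.
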Material /BRCosmetics/Characters/Player/Male/Medium/Bodies/M_MED_SheerWill_Phase/Material/Overlay/M_SoftGlowOutlines
graph {
 properties {
  bUseMaterialAttributes=true
  bUsedWithMeshDeformer=true
  bUsedWithSkeletalMesh=true
  blendMode=BLEND_Translucent
  pixelDepthOffsetMode=PDOM_Legacy
  shadingModel=MSM_Unlit
  translucencyPass=MTP_BeforeDOF
  twoSided=true
  usedShadingModels=Unlit
 }
 nodes [MaterialFunctionCall() {materialFunction=&/MF_Sheerwill_Outlines}]
}
